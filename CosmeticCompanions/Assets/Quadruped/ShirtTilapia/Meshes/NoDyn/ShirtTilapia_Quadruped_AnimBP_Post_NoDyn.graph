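# Post-process style Animation Blueprint for the ShirtTilapia quadruped cosmetic companion, in its NoDyn variant.
# The input pose passes through a face-pose asset blend and DefaultSlot montage playback.
# A weighted constraint then drives the collar tag master bone from neck, head and upper spine bones, favoring the neck.
# A root-relative Anim Dynamics node finally adds damped, angle-limited spring swing to the collar tag, despite the NoDyn name.
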AnimBlueprint /CosmeticCompanions/Assets/Quadruped/ShirtTilapia/Meshes/NoDyn/ShirtTilapia_Quadruped_AnimBP_Post_NoDyn
target_skeleton=/CosmeticCompanions/Assets/Quadruped/ShirtTilapia/Meshes/ShirtTilapia_Quadruped_Skeleton
graphs [
 AnimGraph {nodes [
   n1 = ShirtTilapia_Quadruped_FacePoses_PoseAsset!(SourcePose:Pose Link Structure=Input Pose()) -> (Pose:Pose Link Structure) {
    is=title AnimGraphNode_PoseBlendNode
    blendOption=Linear
    poseAsset=/CosmeticCompanions/Assets/Quadruped/ShirtTilapia/Meshes/ShirtTilapia_Quadruped_FacePoses_PoseAsset
   }
   n2 = "Constraint - dyn_collar_tag_MSTR (PPPPP)"!(
    ComponentPose=Local To Component(LocalPose=Slot 'DefaultSlot'(Source=@n1))
    Alpha=1
    ConstraintWeights_0=0.4911
    ConstraintWeights_1:float=0.3552
    ConstraintWeights_2:float=0.1118
    ConstraintWeights_3:float=0.0296
    ConstraintWeights_4:float=0.0123
   ) -> (Pose) {
    alphaBoolBlend {blendOption=Linear}
    boneToModify {boneName=dyn_collar_tag_MSTR}
    constraintSetup [
     {targetBone {boneName=neck_02}}
     {targetBone {boneName=neck_01}}
     {targetBone {boneName=head}}
     {targetBone {boneName=spine_03}}
     {targetBone {boneName=spine_02}}
    ]
    constraintWeights [1,1,1,1,1]
   }
   n3 = Anim Dynamics - Bone: dyn_collar_tag(ComponentPose=@n2, Alpha=1, ExternalForce="0,0,0") {
    angularDampingOverride=0.9
    angularSpringConstant=1500
    bAngularSpring=true
    bOverrideAngularDamping=true
    bUseGravityOverride=true
    gravityScale=0.35
    simulationSpace=RootRelative
    alphaBoolBlend {blendOption=Linear}
    boundBone {boneName=dyn_collar_tag}
    componentAppliedLinearAccClamp=<3000 3000 3000>
    componentLinearAccScale=<2 2 2>
    componentLinearVelScale=<2 2 2>
    gravityOverride {z=-980}
    physicsBodyDefinitions [{
      boundBone {boneName=dyn_collar_tag}
      boxExtents=<20 10 10>
      constraintSetup {angularLimitsMax=<5 30 -10>, angularLimitsMin=<-5 -30 -45>, angularTarget=<1 0 0>}
      localJointOffset=<6 0 0>
     }]
   }
   Output Pose(Result=Component To Local(ComponentPose=@n3))
  ]}
 EventGraph {links [], nodes []}
]
variables []
